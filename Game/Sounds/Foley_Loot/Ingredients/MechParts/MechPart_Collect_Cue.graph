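# Sound cue for collecting mechanical-part ingredients, routed through the player-interaction sound class with prop-collect attenuation.
# After a short initial delay, a mixer layers the mech-parts collect foley, a quieter loot special effect, and a generic reward sting.
# The reward sting has its own further delay, so it trails the other two layers.
# A concurrency override limits playback to two instances, stopping the oldest, despite an assigned one-maximum concurrency set.
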SoundCue /Game/Sounds/Foley_Loot/Ingredients/MechParts/MechPart_Collect_Cue
root=Delay_0
nodes [
 {id=Delay_0, children [Mixer_0], properties {delayMax=0.5, delayMin=0.5}}
 {id=Delay_1, children [WavePlayer_3], properties {delayMax=1.2, delayMin=1.2}}
 {id=Mixer_0, children [WavePlayer_1,Delay_1,WavePlayer_4], properties {inputVolume [1.5,0.8,0.35]}}
 {id=WavePlayer_1, properties {soundWaveAssetPtr=&/MechParts_Collect_Real03}}
 {id=WavePlayer_3, properties {soundWaveAssetPtr=/Game/Sounds/Fort_UI_Sounds/Gameplay/Reward_Generic}}
 {id=WavePlayer_4, properties {soundWaveAssetPtr=/Game/Sounds/Foley_Loot/Ingredients/Generic/find_loot_special_effect}}
]
properties {
 attenuationSettings=/Game/Sounds/Attenuation/Fort_Prop_Collect_Att
 bOverrideConcurrency=true
 maxDistance=2097152
 soundClassObject=/Game/Sounds/FortSoundClassesAndModes/Fort_08_Player_Interacts
 volumeMultiplier=0.5
 concurrencyOverrides {maxCount=2, resolutionRule=StopOldest}
 concurrencySet [/Game/Sounds/Concurrency_Rules/OneMaximum_Owner_StopLowestPriority]
}
