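# Sound cue for collecting a nuts and bolts ingredient, routed to the loot collect sound class.
# A mixer layers the nuts-and-bolts collect wave with a generic find-loot special effect wave, leaving its middle input empty.
# It uses prop collect attenuation at a reduced overall volume.
# A concurrency override caps it at two instances, stopping the oldest, in place of the assigned one-per-owner concurrency asset.
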SoundCue /Game/Sounds/Foley_Loot/Ingredients/MechParts/NutsBolts_Collect_Cue
root=Mixer_2
nodes [
 {id=Mixer_2, children [WavePlayer_9, null, WavePlayer_10], properties {inputVolume [1,0.6,1]}}
 {id=WavePlayer_10, properties {soundWaveAssetPtr=/Game/Sounds/Foley_Loot/Ingredients/Generic/find_loot_special_effect}}
 {id=WavePlayer_9, properties {soundWaveAssetPtr=&/NutsAndBoltsCollect}}
]
properties {
 attenuationSettings=/Game/Sounds/Attenuation/Fort_Prop_Collect_Att
 bOverrideConcurrency=true
 maxDistance=2097152
 priority=0.5
 soundClassObject=&Fort_Loot_Collect
 volumeMultiplier=0.251
 concurrencyOverrides {maxCount=2, resolutionRule=StopOldest}
 concurrencySet [/Game/Sounds/Concurrency_Rules/OneMaximum_Owner_StopLowestPriority]
}
